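; Gameplay ability for a Season 22 dig quest that spawns quest actors, apparently fakes, at a newly targeted tagged actor while the objective is active or test-tagged.
; It rolls a drop tier, picks a weighted entry, then launches that entry's actors with random yaw and velocity.
; Fail counts give bad-luck protection, raising tier odds and entry weights until selected.
; A test flag instead simulates many tier rolls and prints statistics.
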(event Ability|EventActivateAbilityFromEvent
  (params (EventData :type Gameplay%20Event%20Data%20Structure%20%28by%20ref%29))
  (|CallParentFunction EventData)
  (CallFunction|UpdateStoredTarget EventData)
  (if (Variables|Test|GetTestDropRates)
    (CallFunction|TestRollAllDropEntries (Variables|Test|GetTestDropCount))
    (else
      (bind (_tierindex _selectedtier) (DropTier|RollDropTier))
      (if (> _tierindex -1)
        (bind _selectedentry (DropEntry|RollforDropEntry _tierindex _selectedtier))
        (Spawning|SpawnQuestActorsfromDropEntry _selectedentry EventData)
        (else
          _))))
  (Ability|EndAbility))

(fn BumpDropEntryFailCount
  (params (DropEntry :type Struct%20Spawned%20Quest%20Actor%20Drop%20Entry%20Structure) (EntryIndex :type Integer) (TierIndex :type Integer))
  (bind (_spawnedquestactorclass_6_f15cfde2424412e9badbe2975b317a07 _dropquantity_14_6469e6844f83772bb80669bcf263cb9d _failcount_10_49e24745476acf5406dc15a79201f03e) (Utilities|Struct|BreakStructSpawnedQuestActorDropEntry DropEntry))
  (DropEntry|UpdateDropEntry EntryIndex (Utilities|Struct|MakeStructSpawnedQuestActorDropEntry _spawnedquestactorclass_6_f15cfde2424412e9badbe2975b317a07 _dropquantity_14_6469e6844f83772bb80669bcf263cb9d (+ _failcount_10_49e24745476acf5406dc15a79201f03e 1)) TierIndex))

(fn BumpDropTierFailCount
  (params (DropTier :type Struct%20Spawned%20Quest%20Actor%20Drop%20Tier%20Structure) (Index :type Integer))
  (bind (_dropchance_26_053c214a4411031865d2cfa8f6fd560d _dropentries_27_f15cfde2424412e9badbe2975b317a07 _failcount_10_49e24745476acf5406dc15a79201f03e) (Utilities|Struct|BreakStructSpawnedQuestActorDropTier DropTier))
  (Utilities|Array|SetArrayElem (Variables|Setup|GetDropTiers) Index (Utilities|Struct|MakeStructSpawnedQuestActorDropTier _dropchance_26_053c214a4411031865d2cfa8f6fd560d _dropentries_27_f15cfde2424412e9badbe2975b317a07 (+ _failcount_10_49e24745476acf5406dc15a79201f03e 1))))

(fn GenerateSpawnVelocity
  (return (* (Math|Vector|MakeVector (+ (.x (Math|Vector|GetUpVector "0, 0, 0")) (Math|Random|RandomFloatinRange (Variables|Data|GetSpread) (* (Variables|Data|GetSpread) -1))) (+ (.y (Math|Vector|GetUpVector "0, 0, 0")) (Math|Random|RandomFloatinRange (Variables|Data|GetSpread) (* (Variables|Data|GetSpread) -1))) (+ (.z (Math|Vector|GetUpVector "0, 0, 0")) (Math|Random|RandomFloatinRange (Variables|Data|GetSpread) (* (Variables|Data|GetSpread) -1)))) (Math|Random|RandomFloatinRange (Variables|Data|GetSpeed_Min) (Variables|Data|GetSpeed_Max)))))

(fn ShouldAbilityRespondToEvent
  (params (ActorInfo :type Gameplay%20Ability%20Actor%20Info%20Structure) (Payload :type Gameplay%20Event%20Data%20Structure))
  (bind (_eventtag _instigator _target _optionalobject _optionalobject2 _contexthandle _instigatortags _targettags) (Utilities|Struct|BreakGameplayEventData Payload))
  (if (and (!= (Variables|Data|GetStoredTarget) _target) (GameplayTags|HasAnyTags _targettags (Variables|Setup|GetTargetTags_Any)))
    (bind _asfort_player_pawn (Utilities|Casting|CastToFortPlayerPawn (Ability|GetAvatarActorfromActorInfo))
      (:then
        (return (or (GameplayTags|TagContainer|CompareTagAssetInterfacetoOtherTagContainers (LocalVariables|GetTestModeTag) _asfort_player_pawn) (and (Quest|IsQuestinProgress (Quests|GetQuestManager (Fortnite|GetFortPlayerControllerfromActor _asfort_player_pawn) "Invalid") (Variables|Setup|GetQuestDefinition)) (Quest|IsObjectivewithNameinProgress (Quests|GetQuestManager (Fortnite|GetFortPlayerControllerfromActor _asfort_player_pawn) "Invalid") (Variables|Setup|GetQuestDefinition) (Variables|Setup|GetBackend%20Name) :!class /Script/FortniteGame.FortQuestManager)))))
      (:CastFailed
        (return false)))
    (else
      (return false))))

(fn ResetDropEntryFailCount
  (params (DropEntry :type Struct%20Spawned%20Quest%20Actor%20Drop%20Entry%20Structure) (EntryIndex :type Integer) (TierIndex :type Integer))
  (bind (_spawnedquestactorclass_6_f15cfde2424412e9badbe2975b317a07 _dropquantity_14_6469e6844f83772bb80669bcf263cb9d) (Utilities|Struct|BreakStructSpawnedQuestActorDropEntry DropEntry))
  (DropEntry|UpdateDropEntry EntryIndex (Utilities|Struct|MakeStructSpawnedQuestActorDropEntry _spawnedquestactorclass_6_f15cfde2424412e9badbe2975b317a07 _dropquantity_14_6469e6844f83772bb80669bcf263cb9d) TierIndex))

(fn ResetDropTierFailCount
  (params (DropTier :type Struct%20Spawned%20Quest%20Actor%20Drop%20Tier%20Structure) (Index :type Integer))
  (bind (_dropchance_26_053c214a4411031865d2cfa8f6fd560d _dropentries_27_f15cfde2424412e9badbe2975b317a07) (Utilities|Struct|BreakStructSpawnedQuestActorDropTier DropTier))
  (Utilities|Array|SetArrayElem (Variables|Setup|GetDropTiers) Index (Utilities|Struct|MakeStructSpawnedQuestActorDropTier _dropchance_26_053c214a4411031865d2cfa8f6fd560d _dropentries_27_f15cfde2424412e9badbe2975b317a07)))

(fn RollDropTier
  (LocalVariables|SetHaveSuccessfulRoll false)
  (LocalVariables|SetTempSelectedTier)
  (LocalVariables|SetTempTierIndex -1)
  (bind (_array_element _array_index) (Utilities|Array|ForEachLoopwithBreak (Variables|Setup|GetDropTiers))
    (:LoopBody
      (bind _rollsucceeded (DropTier|RollforNextTier _array_element _array_index))
      (if _rollsucceeded
        _
        (else
          _))
      (LocalVariables|SetTempSelectedTier _array_element)
      (LocalVariables|SetTempTierIndex _array_index))
    (:Completed
      _))
  (return (LocalVariables|GetTempTierIndex) (LocalVariables|GetTempSelectedTier)))

(fn RollForDropEntry
  (params (SlectedTierIndex :type Integer) (SelectedTier :type Struct%20Spawned%20Quest%20Actor%20Drop%20Tier%20Structure))
  (bind (_dropchance_26_053c214a4411031865d2cfa8f6fd560d _dropentries_27_f15cfde2424412e9badbe2975b317a07) (Utilities|Struct|BreakStructSpawnedQuestActorDropTier SelectedTier))
  (bind _tempindex (LocalVariables|GetTempIndex))
  (bind _summedchance (LocalVariables|GetSummedChance))
  (LocalVariables|SetDenominator 0.0)
  (LocalVariables|SetSummedChance 0.0)
  (LocalVariables|SetTempIndex 0)
  (LocalVariables|SetTempDropEntry (Utilities|Array|Get(acopy) _dropentries_27_f15cfde2424412e9badbe2975b317a07))
  (for _array_element _dropentries_27_f15cfde2424412e9badbe2975b317a07
    (LocalVariables|SetDenominator (+ (LocalVariables|GetDenominator) (Math|Conversions|ToFloat(Integer) (+ (.FailCount_10_49E24745476ACF5406DC15A79201F03E (Utilities|Struct|BreakStructSpawnedQuestActorDropEntry _array_element)) 1)))))
  (LocalVariables|SetRoll (Math|Random|RandomFloatinRange 0.0 (LocalVariables|GetDenominator)))
  (bind (_array_element_2 _array_index) (Utilities|Array|ForEachLoopwithBreak _dropentries_27_f15cfde2424412e9badbe2975b317a07)
    (:LoopBody
      (if (< (LocalVariables|GetRoll) (+ _summedchance (Math|Conversions|ToFloat(Integer) (+ (.FailCount_10_49E24745476ACF5406DC15A79201F03E (Utilities|Struct|BreakStructSpawnedQuestActorDropEntry _array_element_2)) 1))))
        (LocalVariables|SetTempDropEntry _array_element_2)
        (LocalVariables|SetTempIndex _array_index))
      (LocalVariables|SetSummedChance (+ _summedchance (Math|Conversions|ToFloat(Integer) (+ (.FailCount_10_49E24745476ACF5406DC15A79201F03E (Utilities|Struct|BreakStructSpawnedQuestActorDropEntry _array_element_2)) 1)))))
    (:Completed
      _))
  (for (_array_element_3 _array_index_2) _dropentries_27_f15cfde2424412e9badbe2975b317a07
    (if (== _array_index_2 _tempindex)
      (DropEntry|ResetDropEntryFailCount _array_element_3 _array_index_2 SlectedTierIndex)
      (else
        (DropEntry|BumpDropEntryFailCount _array_element_3 _array_index_2 SlectedTierIndex))))
  (return (LocalVariables|GetTempDropEntry) _tempindex))

(fn RollForNextTier
  (params (DropTier :type Struct%20Spawned%20Quest%20Actor%20Drop%20Tier%20Structure) (Index :type Integer))
  (bind (_dropchance_26_053c214a4411031865d2cfa8f6fd560d _dropentries_27_f15cfde2424412e9badbe2975b317a07 _failcount_10_49e24745476acf5406dc15a79201f03e) (Utilities|Struct|BreakStructSpawnedQuestActorDropTier DropTier))
  (if (< (Math|Random|RandomFloat) (* (+ 1 _failcount_10_49e24745476acf5406dc15a79201f03e) (Utilities|Map|Find (Variables|Data|GetDropChanceMapping) _dropchance_26_053c214a4411031865d2cfa8f6fd560d)))
    (DropTier|ResetDropTierFailCount DropTier Index)
    (return true)
    (else
      (DropTier|BumpDropTierFailCount DropTier Index)
      (return false))))

(fn SpawnQuestActorsFromDropEntry
  (params (DropEntry :type Struct%20Spawned%20Quest%20Actor%20Drop%20Entry%20Structure) (EventData :type Gameplay%20Event%20Data%20Structure))
  (bind (_spawnedquestactorclass_6_f15cfde2424412e9badbe2975b317a07 _dropquantity_14_6469e6844f83772bb80669bcf263cb9d) (Utilities|Struct|BreakStructSpawnedQuestActorDropEntry DropEntry))
  (bind (_eventtag _instigator _target _optionalobject _optionalobject2 _contexthandle _instigatortags _targettags) (Utilities|Struct|BreakGameplayEventData EventData))
  (bind _owningplayerpawn (Variables|GAVars|Data|GetOwningPlayerPawn))
  (for _ (range 1 (+ _dropquantity_14_6469e6844f83772bb80669bcf263cb9d 1))
    (bind _returnvalue (Building|SpawnBuildingGameplayActor _spawnedquestactorclass_6_f15cfde2424412e9badbe2975b317a07 (Math|Transform|MakeTransform (Transformation|GetActorLocation _target) (Math|Rotator|MakeRotator 0.0 0.0 (Math|Random|RandomFloatinRange 0.0 3.6e+02))) _instigator))
    (bind _asbp_spawned_quest_actor_parent (Utilities|Casting|CastToBP_SpawnedQuestActor_Parent _returnvalue))
    (Variables|/Game/Athena/Items/QuestParents/SpawnedQuestActor/BP_SpawnedQuestActor_Parent.BP_SpawnedQuestActor_Parent_C|SetbTestVisualsNoQuestRequired (GameplayTags|TagContainer|CompareTagAssetInterfacetoOtherTagContainers (LocalVariables|GetTestModeTag) _owningplayerpawn) _asbp_spawned_quest_actor_parent)
    (Projectile|SetVelocity (Variables|/Game/Athena/Items/QuestParents/SpawnedQuestActor/BP_SpawnedQuestActor_Parent.BP_SpawnedQuestActor_Parent_C|GetFortProjectileMovement _asbp_spawned_quest_actor_parent) (Spawning|GenerateSpawnVelocity))
    (Class|BPSpawnedQuestActorParent|CheckCosmetics _asbp_spawned_quest_actor_parent (Fortnite|GetFortPlayerControllerfromActor _owningplayerpawn))
    (Ability|GameplayEffect|AddLooseGameplayTags _asbp_spawned_quest_actor_parent _targettags)))

(fn TestRollAllDropEntries
  (params (TotalDrops :type Integer))
  (bind _rolledtiers (LocalVariables|GetRolledTiers))
  (for _ (range TotalDrops)
    (bind _tierindex (DropTier|RollDropTier))
    (if (> _tierindex -1)
      (bind _result (Math|Integer|IncrementInt (Utilities|Array|Get(acopy) _rolledtiers _tierindex))
        (:%20%20
          (Utilities|Array|SetArrayElem _rolledtiers _tierindex _result true)
          (for _array_element (Variables|Setup|GetDropTiers)
            (if (< (Utilities|Array|Get(acopy) (LocalVariables|GetTierMaxFailCount) _tierindex) (.FailCount_10_49E24745476ACF5406DC15A79201F03E (Utilities|Struct|BreakStructSpawnedQuestActorDropTier _array_element)))
              (Utilities|Array|SetArrayElem (LocalVariables|GetTierMaxFailCount) _tierindex (.FailCount_10_49E24745476ACF5406DC15A79201F03E (Utilities|Struct|BreakStructSpawnedQuestActorDropTier _array_element)) true)))))
      (else
        (Math|Integer|IncrementInt (LocalVariables|GetNoDrops)))))
  (for (_array_element_2 _array_index) (LocalVariables|GetRolledTiers)
    (Development|PrintString (Utilities|String|Append "Drop Tier: " (Utilities|String|ToString(Integer) _array_index) ", Total times exactly this tier: " (Utilities|String|ToString(Integer) _array_element_2) ", Percent of total: " (Utilities|String|ToString(Float) (/ (Math|Conversions|ToFloat(Integer) _array_element_2) (Math|Conversions|ToFloat(Integer) TotalDrops))) ", Max fail count: " (Utilities|String|ToString(Integer) (Utilities|Array|Get(acopy) (LocalVariables|GetTierMaxFailCount) _array_index))) true true "(R=0.000000,G=0.660000,B=1.000000,A=1.000000)" 3e+01))
  (Development|PrintString (Utilities|String|Append "Number of times where there were no drops: " (Utilities|String|ToString(Integer) (LocalVariables|GetNoDrops))) true true "(R=0.000000,G=0.660000,B=1.000000,A=1.000000)" 3e+01))

(fn UpdateDropEntry
  (params (EntryIndex :type Integer) (UpdatedDropEntry :type Struct%20Spawned%20Quest%20Actor%20Drop%20Entry%20Structure) (TierIndex :type Integer))
  (bind _droptiers (Variables|Setup|GetDropTiers))
  (bind _output (Utilities|Array|Get(acopy) _droptiers TierIndex))
  (bind (_dropchance_26_053c214a4411031865d2cfa8f6fd560d _dropentries_27_f15cfde2424412e9badbe2975b317a07 _failcount_10_49e24745476acf5406dc15a79201f03e) (Utilities|Struct|BreakStructSpawnedQuestActorDropTier _output))
  (LocalVariables|SetTempDropEntries _dropentries_27_f15cfde2424412e9badbe2975b317a07)
  (Utilities|Array|SetArrayElem (LocalVariables|GetTempDropEntries) EntryIndex UpdatedDropEntry)
  (Utilities|Array|SetArrayElem (Variables|Setup|GetDropTiers) TierIndex (Utilities|Struct|MakeStructSpawnedQuestActorDropTier _dropchance_26_053c214a4411031865d2cfa8f6fd560d (LocalVariables|GetTempDropEntries) _failcount_10_49e24745476acf5406dc15a79201f03e)))

(fn UpdateStoredTarget
  (params (EventData :type Gameplay%20Event%20Data%20Structure))
  (bind (_eventtag _instigator _target) (Utilities|Struct|BreakGameplayEventData EventData))
  (Variables|Data|SetStoredTarget _target))
